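# Physics asset for the Elf backpack cosmetic, giving its two candy canes and bow simulated secondary motion.
# A capsule on the pack bone anchors the rig, with capsules on the left and right canes and boxes on two bow bones.
# Each cane hangs from the pack with tight swing and twist limits.
# The bow bones form a chain off the pack that swings more loosely toward its end.
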PhysicsAsset /BRCosmetics/Accessories/FORT_Backpacks/Backpack_M_MED_Elf/Meshes/M_MED_Elf_Physics
bodies {
 dyn_bow_01 [{shapeType=Box, extentX=6.8812, extentY=17.6564, extentZ=2, center=<1 0 0>, rotation=<rot 90 0 0>}]
 dyn_bow_02 [{shapeType=Box, extentX=6.8812, extentY=17.6564, extentZ=2, center=<1 0 0>, rotation=<rot 90 0 0>}]
 dyn_cane_l [{shapeType=Capsule, length=27.7997, radius=2.8671, center=<12.9797 0 -2>, rotation=<rot 90 0 0>}]
 dyn_cane_r [{shapeType=Capsule, length=27.7342, radius=2.9072, center=<12.0607 0.0001 -2>, rotation=<rot 90 0 0>}]
 pack [{shapeType=Capsule, length=6.1782, radius=10.0358, center=<0 3.1007 -5.4931>}]
}
constraints [
 {bone1Name=dyn_cane_l, bone2Name=pack, swing1LimitDegrees=5, swing2LimitDegrees=5, twistLimitDegrees=1}
 {bone1Name=dyn_cane_r, bone2Name=pack, swing1LimitDegrees=5, swing2LimitDegrees=5, twistLimitDegrees=1}
 {bone1Name=dyn_bow_01, bone2Name=pack, swing1LimitDegrees=17, swing2LimitDegrees=17, twistLimitDegrees=2}
 {bone1Name=dyn_bow_02, bone2Name=dyn_bow_01, swing1LimitDegrees=25, swing2LimitDegrees=25, twistLimitDegrees=2}
]
